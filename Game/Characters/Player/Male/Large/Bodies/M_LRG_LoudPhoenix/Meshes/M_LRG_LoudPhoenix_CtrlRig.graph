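# Control Rig for the LoudPhoenix large male player body that procedurally drives its armor, piston, pipe and corrective bones from body motion.
# Sequenced branches remap forearm, leg, foot and clavicle Euler angles into weights for self-relative bone offsets.
# Constraints drive deform clavicles and foot pistons, while spherical pose readers handle upper arms (multiplied pairs) and read the head.
# Unlike its left twin, the right upper-arm pipe end offset stays at full weight.
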ControlRigBlueprint /Game/Characters/Player/Male/Large/Bodies/M_LRG_LoudPhoenix/Meshes/M_LRG_LoudPhoenix_CtrlRig
graphs [
 RigVMFunctionLibrary {links [], nodes []}
 RigVMModel {nodes [
   n1 = &RigUnit_PrepareForExecution()
   RigUnit_GetTransform = &RigUnit_GetTransform(Item=(Type=Bone,Name="lowerarm_l"), Item.Name=RigVMVariableNode() {variable=temp_bone_control_1}.Value)
   SetControlOffset = /Script/ControlRig.RigUnit_SetControlOffset(Control=NewControl, Space=GlobalSpace, ExecutePin=@n1.ExecutePin, Offset=@RigUnit_GetTransform.Transform) {is=struct RigVMUnitNode, template="SetControlOffset::Execute(in Control,io Offset,in Space)"}
   > &RigUnit_SetControlOffset(
    Control=NewControl_2
    Space=GlobalSpace
    Offset=&RigUnit_GetTransform(Item=(Type=Bone,Name="lowerarm_l"), Item.Name=RigVMVariableNode() {variable=temp_bone_control_2}.Value).Transform
   ) {template="SetControlOffset::Execute(in Control,io Offset,in Space)"}
   RigUnit_BeginExecution = &RigUnit_BeginExecution()
   SequenceAggregate = RigVMAggregateNode(ExecuteContext=@RigUnit_BeginExecution.ExecutePin)
   SequenceAggregate_1 = &RigVMFunction_Sequence(ExecuteContext=@SequenceAggregate.A)
   &RigUnit_RotationConstraintLocalSpaceOffset(
    AdvancedSettings=(RotationOrderForFilter=ZXY)
    Child=(Type=Bone,Name="dfrm_clavicle_l")
    Filter=(bX=false,bY=True,bZ=True)
    Parents=((Item=(Type=Bone,Name="clavicle_l"),Weight=1))
    Weight=1
    ExecutePin=@SequenceAggregate_1.A
   )
   &RigUnit_RotationConstraintLocalSpaceOffset(
    AdvancedSettings=(RotationOrderForFilter=ZXY)
    Child=(Type=Bone,Name="dfrm_clavicle_r")
    Filter=(bX=false,bY=True,bZ=True)
    Parents=((Item=(Type=Bone,Name="clavicle_r"),Weight=1))
    Weight=1
    ExecutePin=@SequenceAggregate_1.B
   )
   SequenceAggregate_1_1 = &RigVMFunction_Sequence(ExecuteContext=@SequenceAggregate.B)
   n2 = &RigVMFunction_MathQuaternionToEuler(RotationOrder=ZYX, Value=&RigUnit_GetTransform(Item=(Type=Bone,Name="lowerarm_l"), Space=LocalSpace).Transform.Rotation)
   RigUnit_SetRelativeTransformForItem = /Script/ControlRig.RigUnit_SetRelativeTransformForItem(
    Child=(Type=Bone,Name="extra_lowerarm_plate_l")
    Parent=(Type=Bone,Name="extra_lowerarm_plate_l")
    Value=(Rotation=(X=0,Y=0,Z=0,W=1),Translation=(X=14,Y=0,Z=0),Scale3D=(X=1,Y=1,Z=1))
    bParentInitial=False
    bPropagateToChildren=True
    Weight=&RigVMFunction_MathDoubleRemap(SourceMaximum=120, SourceMinimum=52, TargetMaximum=1, TargetMinimum=0, Value=@n2.Result.Z).Result
    ExecutePin=@SequenceAggregate_1_1.A
   ) {
    is=struct RigVMUnitNode
    template="Set Relative Transform::Execute(in Child,in Parent,in bParentInitial,in Value,in Weight,in bPropagateToChildren)"
   }
   > &RigUnit_SetRelativeTransformForItem(
    Child=(Type=Bone,Name="extra_lowerarm_corrective_l")
    Parent=(Type=Bone,Name="extra_lowerarm_corrective_l")
    Value=(Rotation=(X=0,Y=0,Z=0,W=1),Translation=(X=9,Y=0,Z=0),Scale3D=(X=1,Y=1,Z=1))
    bParentInitial=False
    bPropagateToChildren=True
    Weight=&RigVMFunction_MathDoubleRemap(SourceMaximum=136, SourceMinimum=100, TargetMaximum=1, TargetMinimum=0, Value=@n2.Result.Z).Result
   ) {template="Set Relative Transform::Execute(in Child,in Parent,in bParentInitial,in Value,in Weight,in bPropagateToChildren)"}
   n3 = &RigVMFunction_MathQuaternionToEuler(
    RotationOrder=ZYX
    Value=&RigUnit_GetTransform(Item=(Type=Bone,Name="lowerarm_r"), Space=LocalSpace, bInitial=false).Transform.Rotation
   )
   &RigUnit_SetRelativeTransformForItem(
    Child=(Type=Bone,Name="extra_rowerarm_plate_r")
    Parent=(Type=Bone,Name="extra_rowerarm_plate_r")
    Value=(Rotation=(X=0,Y=0,Z=0,W=1),Translation=(X=-14,Y=0,Z=0),Scale3D=(X=1,Y=1,Z=1))
    bParentInitial=False
    bPropagateToChildren=True
    Weight=&RigVMFunction_MathDoubleRemap(SourceMaximum=120, SourceMinimum=52, TargetMaximum=1, TargetMinimum=0, Value=@n3.Result.Z).Result
    ExecutePin=@SequenceAggregate_1_1.B
   ) {template="Set Relative Transform::Execute(in Child,in Parent,in bParentInitial,in Value,in Weight,in bPropagateToChildren)"}
   > &RigUnit_SetRelativeTransformForItem(
    Child=(Type=Bone,Name="extra_lowerarm_corrective_r")
    Parent=(Type=Bone,Name="extra_lowerarm_corrective_r")
    Value=(Rotation=(X=0,Y=0,Z=0,W=1),Translation=(X=-9,Y=0,Z=0),Scale3D=(X=1,Y=1,Z=1))
    bParentInitial=False
    bPropagateToChildren=True
    Weight=&RigVMFunction_MathDoubleRemap(SourceMaximum=136, SourceMinimum=100, TargetMaximum=1, TargetMinimum=0, Value=@n3.Result.Z).Result
   ) {template="Set Relative Transform::Execute(in Child,in Parent,in bParentInitial,in Value,in Weight,in bPropagateToChildren)"}
   SequenceAggregate_1_1_1 = &RigVMFunction_Sequence(ExecuteContext=@SequenceAggregate.C)
   n4 = &RigVMFunction_MathQuaternionToEuler(
    RotationOrder=YXZ
    Value=&RigUnit_GetTransform(Item=(Type=Bone,Name="thigh_r"), Space=LocalSpace, bInitial=false).Transform.Rotation
   )
   &RigUnit_SetRelativeTransformForItem(
    Child=(Type=Bone,Name="extra_pelvis_armor_r")
    Parent=(Type=Bone,Name="extra_pelvis_armor_r")
    Value=(Rotation=(X=0,Y=0,Z=0,W=1),Translation=(X=0,Y=0,Z=0),Scale3D=(X=1,Y=1,Z=1))
    bParentInitial=false
    bPropagateToChildren=True
    Value.Rotation=&RigVMFunction_MathQuaternionFromEuler(Euler=(X=0,Y=-7,Z=0)).Result
    Weight=&RigVMFunction_MathDoubleMax(
     B=&RigVMFunction_MathDoubleRemap(SourceMaximum=-40, SourceMinimum=0, TargetMaximum=1, TargetMinimum=0, bClamp=false, Value=@n4.Result.X).Result
     A=&RigVMFunction_MathDoubleRemap(SourceMaximum=30, SourceMinimum=0, TargetMaximum=1, TargetMinimum=0, bClamp=false, Value=@n4.Result.X).Result
    ).Result
    ExecutePin=@SequenceAggregate_1_1_1.B
   ) {template="Set Relative Transform::Execute(in Child,in Parent,in bParentInitial,in Value,in Weight,in bPropagateToChildren)"}
   > &RigUnit_SetRelativeTransformForItem(
    Child=(Type=Bone,Name="extra_pelvis_armor_r")
    Parent=(Type=Bone,Name="extra_pelvis_armor_r")
    Value=(Rotation=(X=0,Y=0,Z=0,W=1),Translation=(X=0,Y=0,Z=0),Scale3D=(X=1,Y=1,Z=1))
    bParentInitial=false
    bPropagateToChildren=True
    Value.Rotation=&RigVMFunction_MathQuaternionFromEuler(Euler=(X=0,Y=-60,Z=0)).Result
    Weight=&RigVMFunction_MathDoubleRemap(SourceMaximum=-110, SourceMinimum=-7, TargetMaximum=1, TargetMinimum=0, bClamp=false, Value=@n4.Result.Y).Result
   ) {template="Set Relative Transform::Execute(in Child,in Parent,in bParentInitial,in Value,in Weight,in bPropagateToChildren)"}
   Minimum = /Script/RigVM.RigVMFunction_MathDoubleMin(
    B=&RigVMFunction_MathDoubleRemap(SourceMaximum=134.2, SourceMinimum=152.2, TargetMaximum=1, TargetMinimum=0, bClamp=false, Value=@n4.Result.Z).Result
    A=&RigVMFunction_MathDoubleRemap(SourceMaximum=-10, SourceMinimum=0, TargetMaximum=0, TargetMinimum=1, bClamp=false, Value=@n4.Result.Z).Result
   ) {is=struct RigVMUnitNode, template="Minimum::Execute(in A,in B,out Result)"}
   > &RigUnit_SetRelativeTransformForItem(
    Child=(Type=Bone,Name="extra_thigh_armor_r")
    Parent=(Type=Bone,Name="extra_thigh_armor_r")
    Value=(Rotation=(X=0,Y=0,Z=0,W=1),Translation=(X=3.5,Y=0,Z=0),Scale3D=(X=1,Y=1,Z=1))
    bParentInitial=false
    bPropagateToChildren=True
    Weight=@Minimum.Result
   ) {template="Set Relative Transform::Execute(in Child,in Parent,in bParentInitial,in Value,in Weight,in bPropagateToChildren)"}
   n5 = &RigVMFunction_MathQuaternionToEuler(RotationOrder=YXZ, Value=&RigUnit_GetTransform(Item=(Type=Bone,Name="thigh_l"), Space=LocalSpace).Transform.Rotation)
   Maximum = &RigVMFunction_MathDoubleMax(
    A=&RigVMFunction_MathDoubleRemap(SourceMaximum=30, SourceMinimum=0, TargetMaximum=1, TargetMinimum=0, bClamp=false, Value=@n5.Result.X).Result
    B=&RigVMFunction_MathDoubleRemap(SourceMaximum=-40, SourceMinimum=0, TargetMaximum=1, TargetMinimum=0, bClamp=false, Value=@n5.Result.X).Result
   )
   &RigUnit_SetRelativeTransformForItem(
    Child=(Type=Bone,Name="extra_pelvis_armor_l")
    Parent=(Type=Bone,Name="extra_pelvis_armor_l")
    Value=(Rotation=(X=0,Y=0,Z=0,W=1),Translation=(X=0,Y=0,Z=0),Scale3D=(X=1,Y=1,Z=1))
    bParentInitial=false
    bPropagateToChildren=True
    Value.Rotation=&RigVMFunction_MathQuaternionFromEuler(Euler=(X=0,Y=-7,Z=0)).Result
    Weight=@Maximum.Result
    ExecutePin=@SequenceAggregate_1_1_1.A
   ) {template="Set Relative Transform::Execute(in Child,in Parent,in bParentInitial,in Value,in Weight,in bPropagateToChildren)"}
   > &RigUnit_SetRelativeTransformForItem(
    Child=(Type=Bone,Name="extra_pelvis_armor_l")
    Parent=(Type=Bone,Name="extra_pelvis_armor_l")
    Value=(Rotation=(X=0,Y=0,Z=0,W=1),Translation=(X=0,Y=0,Z=0),Scale3D=(X=1,Y=1,Z=1))
    bParentInitial=False
    bPropagateToChildren=True
    Value.Rotation=&RigVMFunction_MathQuaternionFromEuler(Euler=(X=0,Y=-60,Z=0)).Result
    Weight=&RigVMFunction_MathDoubleRemap(SourceMaximum=-110, SourceMinimum=-7, TargetMaximum=1, TargetMinimum=0, bClamp=false, Value=@n5.Result.Y).Result
   ) {template="Set Relative Transform::Execute(in Child,in Parent,in bParentInitial,in Value,in Weight,in bPropagateToChildren)"}
   > &RigUnit_SetRelativeTransformForItem(
    Child=(Type=Bone,Name="extra_thigh_armor_l")
    Parent=(Type=Bone,Name="extra_thigh_armor_l")
    Value=(Rotation=(X=0,Y=0,Z=0,W=1),Translation=(X=-3.5,Y=0,Z=0),Scale3D=(X=1,Y=1,Z=1))
    bParentInitial=False
    bPropagateToChildren=True
    Weight=&RigVMFunction_MathDoubleRemap(SourceMaximum=-44, SourceMinimum=-26, TargetMaximum=1, TargetMinimum=0, Value=@n5.Result.Z).Result
   ) {template="Set Relative Transform::Execute(in Child,in Parent,in bParentInitial,in Value,in Weight,in bPropagateToChildren)"}
   SequenceAggregate_1_1_2 = &RigVMFunction_Sequence(ExecuteContext=@SequenceAggregate.D)
   n6 = &RigVMFunction_MathQuaternionToEuler(
    RotationOrder=ZYX
    Value=&RigUnit_GetTransform(Item=(Type=Bone,Name="calf_r"), Space=LocalSpace, bInitial=false).Transform.Rotation
   )
   &RigUnit_SetRelativeTransformForItem(
    Child=(Type=Bone,Name="extra_calf_armor_r")
    Parent=(Type=Bone,Name="extra_calf_armor_r")
    Value=(Rotation=(X=0,Y=0,Z=0,W=1),Translation=(X=-2,Y=2,Z=0),Scale3D=(X=1,Y=1,Z=1))
    bParentInitial=False
    bPropagateToChildren=True
    Value.Rotation=&RigVMFunction_MathQuaternionFromEuler(Euler=(X=0,Y=0,Z=60)).Result
    Weight=&RigVMFunction_MathDoubleRemap(SourceMaximum=140, SourceMinimum=4.5, TargetMaximum=1, TargetMinimum=0, Value=@n6.Result.Z).Result
    ExecutePin=@SequenceAggregate_1_1_2.B
   ) {template="Set Relative Transform::Execute(in Child,in Parent,in bParentInitial,in Value,in Weight,in bPropagateToChildren)"}
   n7 = &RigVMFunction_MathDoubleRemap(SourceMaximum=65, SourceMinimum=5, TargetMaximum=1, TargetMinimum=0, Value=@n6.Result.Z)
   > &RigUnit_SetRelativeTransformForItem(
    Child=(Type=Bone,Name="extra_calf_wheel_bot_r")
    Parent=(Type=Bone,Name="extra_calf_wheel_bot_r")
    Value=(Rotation=(X=0,Y=0,Z=0,W=1),Translation=(X=1.5,Y=-2,Z=0),Scale3D=(X=0.95,Y=0.95,Z=1))
    bParentInitial=False
    bPropagateToChildren=True
    Weight=@n7.Result
   ) {template="Set Relative Transform::Execute(in Child,in Parent,in bParentInitial,in Value,in Weight,in bPropagateToChildren)"}
   > &RigUnit_SetRelativeTransformForItem(
    Child=(Type=Bone,Name="extra_calf_wheel_top_r")
    Parent=(Type=Bone,Name="extra_calf_wheel_top_r")
    Value=(Rotation=(X=0,Y=0,Z=0,W=1),Translation=(X=4.5,Y=3,Z=0),Scale3D=(X=0.95,Y=0.95,Z=1))
    bParentInitial=False
    bPropagateToChildren=True
    Weight=@n7.Result
   ) {template="Set Relative Transform::Execute(in Child,in Parent,in bParentInitial,in Value,in Weight,in bPropagateToChildren)"}
   > &RigUnit_SetRelativeTransformForItem(
    Child=(Type=Bone,Name="extra_calf_gastank_r")
    Parent=(Type=Bone,Name="extra_calf_gastank_r")
    Value=(Rotation=(X=0,Y=0,Z=0,W=1),Translation=(X=5,Y=2,Z=0),Scale3D=(X=0.95,Y=0.95,Z=1))
    bParentInitial=False
    bPropagateToChildren=True
    Weight=@n7.Result
   ) {template="Set Relative Transform::Execute(in Child,in Parent,in bParentInitial,in Value,in Weight,in bPropagateToChildren)"}
   n8 = &RigVMFunction_MathDoubleRemap(SourceMaximum=151, SourceMinimum=94, TargetMaximum=1, TargetMinimum=0, Value=@n6.Result.Z)
   > &RigUnit_SetRelativeTransformForItem(
    Child=(Type=Bone,Name="extra_calf_corrective_r")
    Parent=(Type=Bone,Name="extra_calf_corrective_r")
    Value=(Rotation=(X=0,Y=0,Z=0,W=1),Translation=(X=11,Y=-5,Z=0),Scale3D=(X=1,Y=1,Z=1))
    bParentInitial=False
    bPropagateToChildren=True
    Weight=@n8.Result
   ) {template="Set Relative Transform::Execute(in Child,in Parent,in bParentInitial,in Value,in Weight,in bPropagateToChildren)"}
   > &RigUnit_SetRelativeTransformForItem(
    Child=(Type=Bone,Name="extra_calf_backplate_r")
    Parent=(Type=Bone,Name="extra_calf_backplate_r")
    Value=(Rotation=(X=0,Y=0,Z=0,W=1),Translation=(X=-10,Y=0,Z=0),Scale3D=(X=1,Y=1,Z=1))
    bParentInitial=False
    bPropagateToChildren=True
    Value.Rotation=&RigVMFunction_MathQuaternionFromEuler(Euler=(X=0,Y=0,Z=-60)).Result
    Weight=@n8.Result
   ) {template="Set Relative Transform::Execute(in Child,in Parent,in bParentInitial,in Value,in Weight,in bPropagateToChildren)"}
   n9 = &RigVMFunction_MathQuaternionToEuler(RotationOrder=ZYX, Value=&RigUnit_GetTransform(Item=(Type=Bone,Name="calf_l"), Space=LocalSpace).Transform.Rotation)
   &RigUnit_SetRelativeTransformForItem(
    Child=(Type=Bone,Name="extra_calf_armor_l")
    Parent=(Type=Bone,Name="extra_calf_armor_l")
    Value=(Rotation=(X=0,Y=0,Z=0,W=1),Translation=(X=2,Y=-2,Z=0),Scale3D=(X=1,Y=1,Z=1))
    bParentInitial=False
    bPropagateToChildren=True
    Value.Rotation=&RigVMFunction_MathQuaternionFromEuler(Euler=(X=0,Y=0,Z=60)).Result
    Weight=&RigVMFunction_MathDoubleRemap(SourceMaximum=140, SourceMinimum=4.5, TargetMaximum=1, TargetMinimum=0, Value=@n9.Result.Z).Result
    ExecutePin=@SequenceAggregate_1_1_2.A
   ) {template="Set Relative Transform::Execute(in Child,in Parent,in bParentInitial,in Value,in Weight,in bPropagateToChildren)"}
   n10 = &RigVMFunction_MathDoubleRemap(SourceMaximum=65, SourceMinimum=5, TargetMaximum=1, TargetMinimum=0, Value=@n9.Result.Z)
   > &RigUnit_SetRelativeTransformForItem(
    Child=(Type=Bone,Name="extra_calf_wheel_bot_l")
    Parent=(Type=Bone,Name="extra_calf_wheel_bot_l")
    Value=(Rotation=(X=0,Y=0,Z=0,W=1),Translation=(X=-1.5,Y=2,Z=0),Scale3D=(X=0.95,Y=0.95,Z=1))
    bParentInitial=False
    bPropagateToChildren=True
    Weight=@n10.Result
   ) {template="Set Relative Transform::Execute(in Child,in Parent,in bParentInitial,in Value,in Weight,in bPropagateToChildren)"}
   > &RigUnit_SetRelativeTransformForItem(
    Child=(Type=Bone,Name="extra_calf_wheel_top_l")
    Parent=(Type=Bone,Name="extra_calf_wheel_top_l")
    Value=(Rotation=(X=0,Y=0,Z=0,W=1),Translation=(X=-4.5,Y=-3,Z=0),Scale3D=(X=0.95,Y=0.95,Z=1))
    bParentInitial=False
    bPropagateToChildren=True
    Weight=@n10.Result
   ) {template="Set Relative Transform::Execute(in Child,in Parent,in bParentInitial,in Value,in Weight,in bPropagateToChildren)"}
   > &RigUnit_SetRelativeTransformForItem(
    Child=(Type=Bone,Name="extra_calf_gastank_l")
    Parent=(Type=Bone,Name="extra_calf_gastank_l")
    Value=(Rotation=(X=0,Y=0,Z=0,W=1),Translation=(X=-5,Y=-2,Z=0),Scale3D=(X=0.95,Y=0.95,Z=1))
    bParentInitial=False
    bPropagateToChildren=True
    Weight=@n10.Result
   ) {template="Set Relative Transform::Execute(in Child,in Parent,in bParentInitial,in Value,in Weight,in bPropagateToChildren)"}
   n11 = &RigVMFunction_MathDoubleRemap(SourceMaximum=151, SourceMinimum=94, TargetMaximum=1, TargetMinimum=0, Value=@n9.Result.Z)
   > &RigUnit_SetRelativeTransformForItem(
    Child=(Type=Bone,Name="extra_calf_corrective_l")
    Parent=(Type=Bone,Name="extra_calf_corrective_l")
    Value=(Rotation=(X=0,Y=0,Z=0,W=1),Translation=(X=-11,Y=5,Z=0),Scale3D=(X=1,Y=1,Z=1))
    bParentInitial=False
    bPropagateToChildren=True
    Weight=@n11.Result
   ) {template="Set Relative Transform::Execute(in Child,in Parent,in bParentInitial,in Value,in Weight,in bPropagateToChildren)"}
   > &RigUnit_SetRelativeTransformForItem(
    Child=(Type=Bone,Name="extra_calf_backplate_l")
    Parent=(Type=Bone,Name="extra_calf_backplate_l")
    Value=(Rotation=(X=0,Y=0,Z=0,W=1),Translation=(X=10,Y=0,Z=0),Scale3D=(X=1,Y=1,Z=1))
    bParentInitial=False
    bPropagateToChildren=True
    Value.Rotation=&RigVMFunction_MathQuaternionFromEuler(Euler=(X=0,Y=0,Z=60)).Result
    Weight=@n11.Result
   ) {template="Set Relative Transform::Execute(in Child,in Parent,in bParentInitial,in Value,in Weight,in bPropagateToChildren)"}
   SequenceAggregate_1_1_2_1 = &RigVMFunction_Sequence(ExecuteContext=@SequenceAggregate.E)
   &RigUnit_ParentConstraint(
    AdvancedSettings=(RotationOrderForFilter=XZY)
    Child=(Type=Bone,Name="extra_foot_piston_l")
    Parents=((Item=(Type=Bone,Name="foot_l"),Weight=1))
    Weight=1
    ExecutePin=@SequenceAggregate_1_1_2_1.A
   )
   > &RigUnit_AimConstraintLocalSpaceOffset(
    AdvancedSettings=(DebugSettings=(bEnabled=false,Scale=10,WorldOffset=(Rotation=(X=0,Y=0,Z=0,W=1),Translation=(X=0,Y=0,Z=0),Scale3D=(X=1,Y=1,Z=1))),RotationOrderForFilter=XZY)
    Child=(Type=Bone,Name="extra_foot_piston_l")
    Filter=(bX=false,bY=True,bZ=True)
    Parents=((Item=(Type=Bone,Name="extra_foot_piston_end_l"),Weight=1))
    Weight=1
    WorldUp=(Target=(X=0,Y=0,Z=1),Kind=Direction,Space=(Type=None,Name="None"))
   )
   n12 = &RigVMFunction_MathQuaternionToEuler(
    RotationOrder=ZYX
    Value=&RigUnit_GetTransform(Item=(Type=Bone,Name="foot_l"), Space=LocalSpace, bInitial=false).Transform.Rotation
   )
   > &RigUnit_SetRelativeTransformForItem(
    Child=(Type=Bone,Name="extra_foot_piston_l")
    Parent=(Type=Bone,Name="extra_foot_piston_l")
    Value=(Rotation=(X=0,Y=0,Z=0,W=1),Translation=(X=1,Y=0,Z=2.5),Scale3D=(X=1,Y=1,Z=1))
    bParentInitial=False
    bPropagateToChildren=True
    Weight=&RigVMFunction_MathDoubleRemap(SourceMaximum=31.2, SourceMinimum=-11.3, TargetMaximum=1, TargetMinimum=0, Value=@n12.Result.X).Result
   ) {template="Set Relative Transform::Execute(in Child,in Parent,in bParentInitial,in Value,in Weight,in bPropagateToChildren)"}
   > &RigUnit_SetRelativeTransformForItem(
    Child=(Type=Bone,Name="extra_foot_piston_l")
    Parent=(Type=Bone,Name="extra_foot_piston_l")
    Value=(Rotation=(X=0,Y=0,Z=0,W=1),Translation=(X=0,Y=0,Z=-3),Scale3D=(X=1,Y=1,Z=1))
    bParentInitial=False
    bPropagateToChildren=True
    Weight=&RigVMFunction_MathDoubleRemap(SourceMaximum=-49.6, SourceMinimum=-11.3, TargetMaximum=1, TargetMinimum=0, Value=@n12.Result.X).Result
   ) {template="Set Relative Transform::Execute(in Child,in Parent,in bParentInitial,in Value,in Weight,in bPropagateToChildren)"}
   &RigUnit_ParentConstraint(
    AdvancedSettings=(RotationOrderForFilter=XZY)
    Child=(Type=Bone,Name="extra_foot_piston_r")
    Parents=((Item=(Type=Bone,Name="foot_r"),Weight=1))
    Weight=1
    ExecutePin=@SequenceAggregate_1_1_2_1.B
   )
   > &RigUnit_AimConstraintLocalSpaceOffset(
    AdvancedSettings=(DebugSettings=(bEnabled=false,Scale=10,WorldOffset=(Rotation=(X=0,Y=0,Z=0,W=1),Translation=(X=0,Y=0,Z=0),Scale3D=(X=1,Y=1,Z=1))),RotationOrderForFilter=XZY)
    Child=(Type=Bone,Name="extra_foot_piston_r")
    Filter=(bX=false,bY=True,bZ=True)
    Parents=((Item=(Type=Bone,Name="extra_foot_piston_end_r"),Weight=1))
    Weight=1
    WorldUp=(Target=(X=0,Y=0,Z=1),Kind=Direction,Space=(Type=None,Name="None"))
   )
   n13 = &RigVMFunction_MathQuaternionToEuler(
    RotationOrder=ZYX
    Value=&RigUnit_GetTransform(Item=(Type=Bone,Name="foot_r"), Space=LocalSpace, bInitial=false).Transform.Rotation
   )
   > &RigUnit_SetRelativeTransformForItem(
    Child=(Type=Bone,Name="extra_foot_piston_r")
    Parent=(Type=Bone,Name="extra_foot_piston_r")
    Value=(Rotation=(X=0,Y=0,Z=0,W=1),Translation=(X=-1,Y=0,Z=-2.5),Scale3D=(X=1,Y=1,Z=1))
    bParentInitial=False
    bPropagateToChildren=True
    Weight=&RigVMFunction_MathDoubleRemap(SourceMaximum=31.2, SourceMinimum=-11.3, TargetMaximum=1, TargetMinimum=0, Value=@n13.Result.X).Result
   ) {template="Set Relative Transform::Execute(in Child,in Parent,in bParentInitial,in Value,in Weight,in bPropagateToChildren)"}
   > &RigUnit_SetRelativeTransformForItem(
    Child=(Type=Bone,Name="extra_foot_piston_r")
    Parent=(Type=Bone,Name="extra_foot_piston_r")
    Value=(Rotation=(X=0,Y=0,Z=0,W=1),Translation=(X=0,Y=0,Z=3),Scale3D=(X=1,Y=1,Z=1))
    bParentInitial=False
    bPropagateToChildren=True
    Weight=&RigVMFunction_MathDoubleRemap(SourceMaximum=-49.6, SourceMinimum=-11.3, TargetMaximum=1, TargetMinimum=0, Value=@n13.Result.X).Result
   ) {template="Set Relative Transform::Execute(in Child,in Parent,in bParentInitial,in Value,in Weight,in bPropagateToChildren)"}
   SequenceAggregate_1_1_2_1_1 = &RigVMFunction_Sequence(ExecuteContext=@SequenceAggregate.F)
   n14 = &RigVMFunction_MathQuaternionToEuler(
    RotationOrder=ZXY
    Value=&RigUnit_GetTransform(Item=(Type=Bone,Name="dfrm_clavicle_l"), Space=LocalSpace, bInitial=false).Transform.Rotation
   )
   n15 = &RigVMFunction_MathDoubleRemap(SourceMaximum=25, SourceMinimum=-3, TargetMaximum=1, TargetMinimum=0, Value=@n14.Result.Z)
   &RigUnit_SetRelativeTransformForItem(
    Child=(Type=Bone,Name="extra_front_plate_l")
    Parent=(Type=Bone,Name="extra_front_plate_l")
    Value=(Rotation=(X=0,Y=0,Z=0,W=1),Translation=(X=-0.5,Y=0,Z=0),Scale3D=(X=1,Y=1,Z=1))
    bParentInitial=False
    bPropagateToChildren=True
    Value.Rotation=&RigVMFunction_MathQuaternionFromEuler(Euler=(X=0,Y=0,Z=8)).Result
    Weight=@n15.Result
    ExecutePin=@SequenceAggregate_1_1_2_1_1.A
   ) {template="Set Relative Transform::Execute(in Child,in Parent,in bParentInitial,in Value,in Weight,in bPropagateToChildren)"}
   > &RigUnit_SetRelativeTransformForItem(
    Child=(Type=Bone,Name="extra_top_front_plate_l")
    Parent=(Type=Bone,Name="extra_top_front_plate_l")
    Value=(Rotation=(X=0,Y=0,Z=0,W=1),Translation=(X=0,Y=2,Z=0),Scale3D=(X=1,Y=1,Z=1))
    bParentInitial=False
    bPropagateToChildren=True
    Value.Rotation=&RigVMFunction_MathQuaternionFromEuler(Euler=(X=0,Y=-3,Z=5)).Result
    Weight=@n15.Result
   ) {template="Set Relative Transform::Execute(in Child,in Parent,in bParentInitial,in Value,in Weight,in bPropagateToChildren)"}
   > &RigUnit_SetRelativeTransformForItem(
    Child=(Type=Bone,Name="extra_top_back_plate_l")
    Parent=(Type=Bone,Name="extra_top_back_plate_l")
    Value=(Rotation=(X=0,Y=0,Z=0,W=1),Translation=(X=1,Y=0,Z=0),Scale3D=(X=1,Y=1,Z=1))
    bParentInitial=False
    bPropagateToChildren=True
    Value.Rotation=&RigVMFunction_MathQuaternionFromEuler(Euler=(X=0,Y=0,Z=10)).Result
    Weight=@n15.Result
   ) {template="Set Relative Transform::Execute(in Child,in Parent,in bParentInitial,in Value,in Weight,in bPropagateToChildren)"}
   n16 = &RigVMFunction_MathDoubleRemap(SourceMaximum=-30, SourceMinimum=-4, TargetMaximum=1, TargetMinimum=0, Value=@n14.Result.Z)
   > &RigUnit_SetRelativeTransformForItem(
    Child=(Type=Bone,Name="extra_front_plate_l")
    Parent=(Type=Bone,Name="extra_front_plate_l")
    Value=(Rotation=(X=0,Y=0,Z=0,W=1),Translation=(X=2,Y=0,Z=0),Scale3D=(X=1,Y=1,Z=1))
    bParentInitial=False
    bPropagateToChildren=True
    Value.Rotation=&RigVMFunction_MathQuaternionFromEuler(Euler=(X=0,Y=0,Z=-5)).Result
    Weight=@n16.Result
   ) {template="Set Relative Transform::Execute(in Child,in Parent,in bParentInitial,in Value,in Weight,in bPropagateToChildren)"}
   > &RigUnit_SetRelativeTransformForItem(
    Child=(Type=Bone,Name="extra_top_front_plate_l")
    Parent=(Type=Bone,Name="extra_top_front_plate_l")
    Value=(Rotation=(X=0,Y=0,Z=0,W=1),Translation=(X=-0.5,Y=0,Z=0),Scale3D=(X=1,Y=1,Z=1))
    bParentInitial=False
    bPropagateToChildren=True
    Value.Rotation=&RigVMFunction_MathQuaternionFromEuler(Euler=(X=0,Y=0,Z=-8)).Result
    Weight=@n16.Result
   ) {template="Set Relative Transform::Execute(in Child,in Parent,in bParentInitial,in Value,in Weight,in bPropagateToChildren)"}
   > &RigUnit_SetRelativeTransformForItem(
    Child=(Type=Bone,Name="extra_top_back_plate_l")
    Parent=(Type=Bone,Name="extra_top_back_plate_l")
    Value=(Rotation=(X=0,Y=0,Z=0,W=1),Translation=(X=0,Y=0,Z=0),Scale3D=(X=1,Y=1,Z=1))
    bParentInitial=False
    bPropagateToChildren=True
    Value.Rotation=&RigVMFunction_MathQuaternionFromEuler(Euler=(X=0,Y=0,Z=-7)).Result
    Weight=@n16.Result
   ) {template="Set Relative Transform::Execute(in Child,in Parent,in bParentInitial,in Value,in Weight,in bPropagateToChildren)"}
   n17 = &RigVMFunction_MathDoubleRemap(SourceMaximum=60, SourceMinimum=80, TargetMaximum=1, TargetMinimum=0, Value=@n14.Result.Y)
   > &RigUnit_SetRelativeTransformForItem(
    Child=(Type=Bone,Name="extra_top_front_plate_l")
    Parent=(Type=Bone,Name="extra_top_front_plate_l")
    Value=(Rotation=(X=0,Y=0,Z=0,W=1),Translation=(X=0,Y=0,Z=0),Scale3D=(X=1,Y=1,Z=1))
    bParentInitial=False
    bPropagateToChildren=True
    Value.Rotation=&RigVMFunction_MathQuaternionFromEuler(Euler=(X=0,Y=-15,Z=0)).Result
    Weight=@n17.Result
   ) {template="Set Relative Transform::Execute(in Child,in Parent,in bParentInitial,in Value,in Weight,in bPropagateToChildren)"}
   > &RigUnit_SetRelativeTransformForItem(
    Child=(Type=Bone,Name="extra_top_back_plate_l")
    Parent=(Type=Bone,Name="extra_top_back_plate_l")
    Value=(Rotation=(X=0,Y=0,Z=0,W=1),Translation=(X=0,Y=0,Z=0.5),Scale3D=(X=1,Y=1,Z=1))
    bParentInitial=False
    bPropagateToChildren=True
    Value.Rotation=&RigVMFunction_MathQuaternionFromEuler(Euler=(X=15,Y=-22,Z=0)).Result
    Weight=@n17.Result
   ) {template="Set Relative Transform::Execute(in Child,in Parent,in bParentInitial,in Value,in Weight,in bPropagateToChildren)"}
   n18 = &RigVMFunction_MathQuaternionToEuler(
    RotationOrder=ZXY
    Value=&RigUnit_GetTransform(Item=(Type=Bone,Name="dfrm_clavicle_r"), Space=LocalSpace, bInitial=false).Transform.Rotation
   )
   n19 = &RigVMFunction_MathDoubleMin(
    A=&RigVMFunction_MathDoubleRemap(SourceMaximum=-152, SourceMinimum=-180, TargetMaximum=1, TargetMinimum=0, Value=@n18.Result.Z).Result
    B=&RigVMFunction_MathDoubleRemap(SourceMaximum=0, SourceMinimum=10, TargetMaximum=1, TargetMinimum=0, Value=@n18.Result.Z).Result
   ) {template="Minimum::Execute(in A,in B,out Result)"}
   &RigUnit_SetRelativeTransformForItem(
    Child=(Type=Bone,Name="extra_front_plate_r")
    Parent=(Type=Bone,Name="extra_front_plate_r")
    Value=(Rotation=(X=0,Y=0,Z=0,W=1),Translation=(X=0.5,Y=0,Z=0),Scale3D=(X=1,Y=1,Z=1))
    bParentInitial=False
    bPropagateToChildren=True
    Value.Rotation=&RigVMFunction_MathQuaternionFromEuler(Euler=(X=0,Y=0,Z=8)).Result
    ExecutePin=@SequenceAggregate_1_1_2_1_1.B
    Weight=@n19.Result
   ) {template="Set Relative Transform::Execute(in Child,in Parent,in bParentInitial,in Value,in Weight,in bPropagateToChildren)"}
   > &RigUnit_SetRelativeTransformForItem(
    Child=(Type=Bone,Name="extra_top_front_plate_r")
    Parent=(Type=Bone,Name="extra_top_front_plate_r")
    Value=(Rotation=(X=0,Y=0,Z=0,W=1),Translation=(X=0,Y=-2,Z=0),Scale3D=(X=1,Y=1,Z=1))
    bParentInitial=False
    bPropagateToChildren=True
    Value.Rotation=&RigVMFunction_MathQuaternionFromEuler(Euler=(X=0,Y=-3,Z=5)).Result
    Weight=@n19.Result
   ) {template="Set Relative Transform::Execute(in Child,in Parent,in bParentInitial,in Value,in Weight,in bPropagateToChildren)"}
   > &RigUnit_SetRelativeTransformForItem(
    Child=(Type=Bone,Name="extra_top_back_plate_r")
    Parent=(Type=Bone,Name="extra_top_back_plate_r")
    Value=(Rotation=(X=0,Y=0,Z=0,W=1),Translation=(X=-1,Y=0,Z=0),Scale3D=(X=1,Y=1,Z=1))
    bParentInitial=False
    bPropagateToChildren=True
    Value.Rotation=&RigVMFunction_MathQuaternionFromEuler(Euler=(X=0,Y=0,Z=10)).Result
    Weight=@n19.Result
   ) {template="Set Relative Transform::Execute(in Child,in Parent,in bParentInitial,in Value,in Weight,in bPropagateToChildren)"}
   n20 = &RigVMFunction_MathDoubleMin(
    B=&RigVMFunction_MathDoubleRemap(SourceMaximum=154, SourceMinimum=180, TargetMaximum=1, TargetMinimum=0, Value=@n18.Result.Z).Result
    A=&RigVMFunction_MathDoubleRemap(SourceMaximum=0, SourceMinimum=10, TargetMaximum=0, TargetMinimum=1, Value=@n18.Result.Z).Result
   ) {template="Minimum::Execute(in A,in B,out Result)"}
   > &RigUnit_SetRelativeTransformForItem(
    Child=(Type=Bone,Name="extra_front_plate_r")
    Parent=(Type=Bone,Name="extra_front_plate_r")
    Value=(Rotation=(X=0,Y=0,Z=0,W=1),Translation=(X=-2,Y=0,Z=0),Scale3D=(X=1,Y=1,Z=1))
    bParentInitial=False
    bPropagateToChildren=True
    Value.Rotation=&RigVMFunction_MathQuaternionFromEuler(Euler=(X=0,Y=0,Z=-5)).Result
    Weight=@n20.Result
   ) {template="Set Relative Transform::Execute(in Child,in Parent,in bParentInitial,in Value,in Weight,in bPropagateToChildren)"}
   > &RigUnit_SetRelativeTransformForItem(
    Child=(Type=Bone,Name="extra_top_front_plate_r")
    Parent=(Type=Bone,Name="extra_top_front_plate_r")
    Value=(Rotation=(X=0,Y=0,Z=0,W=1),Translation=(X=0.5,Y=0,Z=0),Scale3D=(X=1,Y=1,Z=1))
    bParentInitial=False
    bPropagateToChildren=True
    Value.Rotation=&RigVMFunction_MathQuaternionFromEuler(Euler=(X=0,Y=0,Z=-8)).Result
    Weight=@n20.Result
   ) {template="Set Relative Transform::Execute(in Child,in Parent,in bParentInitial,in Value,in Weight,in bPropagateToChildren)"}
   > &RigUnit_SetRelativeTransformForItem(
    Child=(Type=Bone,Name="extra_top_back_plate_r")
    Parent=(Type=Bone,Name="extra_top_back_plate_r")
    Value=(Rotation=(X=0,Y=0,Z=0,W=1),Translation=(X=0,Y=0,Z=0),Scale3D=(X=1,Y=1,Z=1))
    bParentInitial=False
    bPropagateToChildren=True
    Value.Rotation=&RigVMFunction_MathQuaternionFromEuler(Euler=(X=0,Y=0,Z=-7)).Result
    Weight=@n20.Result
   ) {template="Set Relative Transform::Execute(in Child,in Parent,in bParentInitial,in Value,in Weight,in bPropagateToChildren)"}
   n21 = &RigVMFunction_MathDoubleRemap(SourceMaximum=-60, SourceMinimum=-80, TargetMaximum=1, TargetMinimum=0, Value=@n18.Result.Y)
   > &RigUnit_SetRelativeTransformForItem(
    Child=(Type=Bone,Name="extra_top_front_plate_r")
    Parent=(Type=Bone,Name="extra_top_front_plate_r")
    Value=(Rotation=(X=0,Y=0,Z=0,W=1),Translation=(X=0,Y=0,Z=0),Scale3D=(X=1,Y=1,Z=1))
    bParentInitial=False
    bPropagateToChildren=True
    Value.Rotation=&RigVMFunction_MathQuaternionFromEuler(Euler=(X=0,Y=-15,Z=0)).Result
    Weight=@n21.Result
   ) {template="Set Relative Transform::Execute(in Child,in Parent,in bParentInitial,in Value,in Weight,in bPropagateToChildren)"}
   > &RigUnit_SetRelativeTransformForItem(
    Child=(Type=Bone,Name="extra_top_back_plate_r")
    Parent=(Type=Bone,Name="extra_top_back_plate_r")
    Value=(Rotation=(X=0,Y=0,Z=0,W=1),Translation=(X=0,Y=0,Z=-0.5),Scale3D=(X=1,Y=1,Z=1))
    bParentInitial=False
    bPropagateToChildren=True
    Value.Rotation=&RigVMFunction_MathQuaternionFromEuler(Euler=(X=15,Y=-22,Z=0)).Result
    Weight=@n21.Result
   ) {template="Set Relative Transform::Execute(in Child,in Parent,in bParentInitial,in Value,in Weight,in bPropagateToChildren)"}
   SequenceAggregate_1_1_2_1_1_1 = &RigVMFunction_Sequence(ExecuteContext=@SequenceAggregate.G)
   n22 = &RigUnit_SphericalPoseReader(
    ActiveRegionScaleFactors=(PositiveWidth=1,NegativeWidth=1,PositiveHeight=1,NegativeHeight=1)
    ActiveRegionSize=0.25
    Debug=(bDrawDebug=false,bDraw2D=false,bDrawLocalAxes=false,DebugScale=25,DebugSegments=20,DebugThickness=0.25)
    DriverAxis=(X=1,Y=0,Z=0)
    DriverItem=(Type=Bone,Name="upperarm_l")
    FalloffRegionScaleFactors=(PositiveWidth=1,NegativeWidth=1,PositiveHeight=1,NegativeHeight=1)
    FalloffSize=0.4
    OptionalParentItem=(Type=Bone,Name="clavicle_l")
    RotationOffset=(X=0,Y=-60,Z=0)
    ExecutePin=@SequenceAggregate_1_1_2_1_1_1.A
   )
   > n23 = &RigUnit_SphericalPoseReader(
    ActiveRegionScaleFactors=(PositiveWidth=1,NegativeWidth=1,PositiveHeight=1,NegativeHeight=1)
    ActiveRegionSize=0.15
    Debug=(bDrawDebug=false,bDraw2D=false,bDrawLocalAxes=false,DebugScale=25,DebugSegments=20,DebugThickness=0.25)
    DriverAxis=(X=0,Y=0,Z=1)
    DriverItem=(Type=Bone,Name="upperarm_l")
    FalloffRegionScaleFactors=(PositiveWidth=1,NegativeWidth=1,PositiveHeight=1,NegativeHeight=1)
    FalloffSize=0.3
    OptionalParentItem=(Type=Bone,Name="clavicle_l")
    RotationOffset=(X=-90,Y=-45,Z=-15)
   )
   Multiply = &RigVMFunction_MathDoubleMul(A=@n23.OutputParam, B=@n22.OutputParam)
   > &RigUnit_SetRelativeTransformForItem(
    Child=(Type=Bone,Name="extra_upperarm_pipe_l")
    Parent=(Type=Bone,Name="extra_upperarm_pipe_l")
    Value=(Rotation=(X=0,Y=0,Z=0,W=1),Translation=(X=10,Y=0,Z=-1),Scale3D=(X=1,Y=1,Z=1))
    bParentInitial=False
    bPropagateToChildren=True
    Value.Rotation=&RigVMFunction_MathQuaternionFromEuler(Euler=(X=0,Y=19,Z=0)).Result
    Weight=@Multiply.Result
   ) {template="Set Relative Transform::Execute(in Child,in Parent,in bParentInitial,in Value,in Weight,in bPropagateToChildren)"}
   > &RigUnit_SetRelativeTransformForItem(
    Child=(Type=Bone,Name="extra_upperarm_pipe_end_l")
    Parent=(Type=Bone,Name="extra_upperarm_pipe_end_l")
    Value=(Rotation=(X=0,Y=0,Z=0,W=1),Translation=(X=4.5,Y=0,Z=0),Scale3D=(X=1,Y=1,Z=1))
    bParentInitial=False
    bPropagateToChildren=True
    Weight=@Multiply.Result
   ) {template="Set Relative Transform::Execute(in Child,in Parent,in bParentInitial,in Value,in Weight,in bPropagateToChildren)"}
   > &RigUnit_SetRelativeTransformForItem(
    Child=(Type=Bone,Name="extra_upperarm_front_armor_l")
    Parent=(Type=Bone,Name="extra_upperarm_front_armor_l")
    Value=(Rotation=(X=0,Y=0,Z=0,W=1),Translation=(X=4,Y=0,Z=0),Scale3D=(X=1,Y=1,Z=1))
    bParentInitial=False
    bPropagateToChildren=True
    Weight=@Multiply.Result
   ) {template="Set Relative Transform::Execute(in Child,in Parent,in bParentInitial,in Value,in Weight,in bPropagateToChildren)"}
   > &RigUnit_SetRelativeTransformForItem(
    Child=(Type=Bone,Name="extra_upperarm_side_armor_l")
    Parent=(Type=Bone,Name="extra_upperarm_side_armor_l")
    Value=(Rotation=(X=0,Y=0,Z=0,W=1),Translation=(X=8,Y=0,Z=0),Scale3D=(X=1,Y=1,Z=1))
    bParentInitial=False
    bPropagateToChildren=True
    Weight=@Multiply.Result
   ) {template="Set Relative Transform::Execute(in Child,in Parent,in bParentInitial,in Value,in Weight,in bPropagateToChildren)"}
   > &RigUnit_SetRelativeTransformForItem(
    Child=(Type=Bone,Name="extra_upperarm_top_armor_l")
    Parent=(Type=Bone,Name="extra_upperarm_top_armor_l")
    Value=(Rotation=(X=0,Y=0,Z=0,W=1),Translation=(X=10,Y=0,Z=-1),Scale3D=(X=1,Y=1,Z=1))
    bParentInitial=False
    bPropagateToChildren=True
    Weight=@Multiply.Result
   ) {template="Set Relative Transform::Execute(in Child,in Parent,in bParentInitial,in Value,in Weight,in bPropagateToChildren)"}
   > &RigUnit_SetRelativeTransformForItem(
    Child=(Type=Bone,Name="extra_upperarm_top_tweak_l")
    Parent=(Type=Bone,Name="extra_upperarm_top_tweak_l")
    Value=(Rotation=(X=0,Y=0,Z=0,W=1),Translation=(X=-1.8,Y=0,Z=0),Scale3D=(X=1,Y=1,Z=1))
    bParentInitial=False
    bPropagateToChildren=True
    Weight=@Multiply.Result
   ) {template="Set Relative Transform::Execute(in Child,in Parent,in bParentInitial,in Value,in Weight,in bPropagateToChildren)"}
   n24 = &RigUnit_SphericalPoseReader(
    ActiveRegionScaleFactors=(PositiveWidth=1,NegativeWidth=1,PositiveHeight=1,NegativeHeight=1)
    ActiveRegionSize=0.25
    Debug=(bDrawDebug=false,bDraw2D=false,bDrawLocalAxes=false,DebugScale=25,DebugSegments=20,DebugThickness=0.25)
    DriverAxis=(X=1,Y=0,Z=0)
    DriverItem=(Type=Bone,Name="upperarm_r")
    FalloffRegionScaleFactors=(PositiveWidth=1,NegativeWidth=1,PositiveHeight=1,NegativeHeight=1)
    FalloffSize=0.4
    OptionalParentItem=(Type=Bone,Name="clavicle_r")
    RotationOffset=(X=0,Y=-60,Z=0)
    ExecutePin=@SequenceAggregate_1_1_2_1_1_1.B
   )
   > n25 = &RigUnit_SphericalPoseReader(
    ActiveRegionScaleFactors=(PositiveWidth=1,NegativeWidth=1,PositiveHeight=1,NegativeHeight=1)
    ActiveRegionSize=0.15
    Debug=(bDrawDebug=false,bDraw2D=false,bDrawLocalAxes=false,DebugScale=25,DebugSegments=20,DebugThickness=0.25)
    DriverAxis=(X=0,Y=0,Z=1)
    DriverItem=(Type=Bone,Name="upperarm_r")
    FalloffRegionScaleFactors=(PositiveWidth=1,NegativeWidth=1,PositiveHeight=1,NegativeHeight=1)
    FalloffSize=0.3
    OptionalParentItem=(Type=Bone,Name="clavicle_r")
    RotationOffset=(X=-90,Y=-45,Z=-15)
   )
   n26 = &RigVMFunction_MathDoubleMul(A=@n25.OutputParam, B=@n24.OutputParam)
   > &RigUnit_SetRelativeTransformForItem(
    Child=(Type=Bone,Name="extra_upperarm_pipe_r")
    Parent=(Type=Bone,Name="extra_upperarm_pipe_r")
    Value=(Rotation=(X=0,Y=0,Z=0,W=1),Translation=(X=-10,Y=0,Z=1),Scale3D=(X=1,Y=1,Z=1))
    bParentInitial=False
    bPropagateToChildren=True
    Value.Rotation=&RigVMFunction_MathQuaternionFromEuler(Euler=(X=0,Y=19,Z=0)).Result
    Weight=@n26.Result
   ) {template="Set Relative Transform::Execute(in Child,in Parent,in bParentInitial,in Value,in Weight,in bPropagateToChildren)"}
   > &RigUnit_SetRelativeTransformForItem(
    Child=(Type=Bone,Name="extra_upperarm_pipe_end_r")
    Parent=(Type=Bone,Name="extra_upperarm_pipe_end_r")
    Value=(Rotation=(X=0,Y=0,Z=0,W=1),Translation=(X=-4.5,Y=0,Z=0),Scale3D=(X=1,Y=1,Z=1))
    Weight=1
    bParentInitial=False
    bPropagateToChildren=True
   ) {template="Set Relative Transform::Execute(in Child,in Parent,in bParentInitial,in Value,in Weight,in bPropagateToChildren)"}
   > &RigUnit_SetRelativeTransformForItem(
    Child=(Type=Bone,Name="extra_upperarm_front_armor_r")
    Parent=(Type=Bone,Name="extra_upperarm_front_armor_r")
    Value=(Rotation=(X=0,Y=0,Z=0,W=1),Translation=(X=-4,Y=0,Z=0),Scale3D=(X=1,Y=1,Z=1))
    bParentInitial=False
    bPropagateToChildren=True
    Weight=@n26.Result
   ) {template="Set Relative Transform::Execute(in Child,in Parent,in bParentInitial,in Value,in Weight,in bPropagateToChildren)"}
   > &RigUnit_SetRelativeTransformForItem(
    Child=(Type=Bone,Name="extra_upperarm_side_armor_r")
    Parent=(Type=Bone,Name="extra_upperarm_side_armor_r")
    Value=(Rotation=(X=0,Y=0,Z=0,W=1),Translation=(X=-8,Y=0,Z=0),Scale3D=(X=1,Y=1,Z=1))
    bParentInitial=False
    bPropagateToChildren=True
    Weight=@n26.Result
   ) {template="Set Relative Transform::Execute(in Child,in Parent,in bParentInitial,in Value,in Weight,in bPropagateToChildren)"}
   > &RigUnit_SetRelativeTransformForItem(
    Child=(Type=Bone,Name="extra_upperarm_top_armor_r")
    Parent=(Type=Bone,Name="extra_upperarm_top_armor_r")
    Value=(Rotation=(X=0,Y=0,Z=0,W=1),Translation=(X=-10,Y=0,Z=1),Scale3D=(X=1,Y=1,Z=1))
    bParentInitial=False
    bPropagateToChildren=True
    Weight=@n26.Result
   ) {template="Set Relative Transform::Execute(in Child,in Parent,in bParentInitial,in Value,in Weight,in bPropagateToChildren)"}
   > &RigUnit_SetRelativeTransformForItem(
    Child=(Type=Bone,Name="extra_upperarm_top_tweak_r")
    Parent=(Type=Bone,Name="extra_upperarm_top_tweak_r")
    Value=(Rotation=(X=0,Y=0,Z=0,W=1),Translation=(X=1.8,Y=0,Z=0),Scale3D=(X=1,Y=1,Z=1))
    bParentInitial=False
    bPropagateToChildren=True
    Weight=@n26.Result
   ) {template="Set Relative Transform::Execute(in Child,in Parent,in bParentInitial,in Value,in Weight,in bPropagateToChildren)"}
   n27 = &RigUnit_SphericalPoseReader(
    ActiveRegionScaleFactors=(PositiveWidth=3,NegativeWidth=3,PositiveHeight=1,NegativeHeight=1)
    ActiveRegionSize=0.1
    Debug=(bDrawDebug=false,bDraw2D=false,bDrawLocalAxes=false,DebugScale=25,DebugSegments=20,DebugThickness=0.25)
    DriverAxis=(X=1,Y=0,Z=0)
    DriverItem=(Type=Bone,Name="head")
    FalloffRegionScaleFactors=(PositiveWidth=3,NegativeWidth=3,PositiveHeight=0.25,NegativeHeight=0.25)
    FalloffSize=0.35
    OptionalParentItem=(Type=Bone,Name="spine_05")
    RotationOffset=(X=-57,Y=-90,Z=0)
    ExecutePin=@SequenceAggregate.I
   )
   > &RigUnit_SetRelativeTransformForItem(
    Child=(Type=Bone,Name="extra_chest_bar")
    Parent=(Type=Bone,Name="extra_chest_bar")
    Value=(Rotation=(X=0,Y=0,Z=0,W=1),Translation=(X=0,Y=0,Z=0),Scale3D=(X=1,Y=1,Z=1))
    bParentInitial=False
    bPropagateToChildren=True
    Value.Rotation=&RigVMFunction_MathQuaternionFromEuler(Euler=(X=0,Y=0,Z=-10)).Result
    Weight=@n27.OutputParam
   ) {template="Set Relative Transform::Execute(in Child,in Parent,in bParentInitial,in Value,in Weight,in bPropagateToChildren)"}
   SequenceAggregate_1_1_2_1_1_1_1 = &RigVMFunction_Sequence(ExecuteContext=@SequenceAggregate.J)
   n28 = &RigUnit_SphericalPoseReader(
    ActiveRegionScaleFactors=(PositiveWidth=1,NegativeWidth=1,PositiveHeight=1.3,NegativeHeight=1)
    ActiveRegionSize=0.1
    Debug=(bDrawDebug=false,bDraw2D=false,bDrawLocalAxes=false,DebugScale=25,DebugSegments=20,DebugThickness=0.25)
    DriverAxis=(X=-1,Y=0,Z=0)
    DriverItem=(Type=Bone,Name="upperarm_l")
    FalloffRegionScaleFactors=(PositiveWidth=1,NegativeWidth=1,PositiveHeight=0.9,NegativeHeight=1)
    FalloffSize=0.2
    OptionalParentItem=(Type=Bone,Name="spine_05")
    RotationOffset=(X=-60,Y=30,Z=40)
    ExecutePin=@SequenceAggregate_1_1_2_1_1_1_1.A
   )
   > &RigUnit_SetRelativeTransformForItem(
    Child=(Type=Bone,Name="extra_front_plate_windshield_l")
    Parent=(Type=Bone,Name="extra_front_plate_windshield_l")
    Value=(Rotation=(X=0,Y=0,Z=0,W=1),Translation=(X=1,Y=0,Z=0),Scale3D=(X=1,Y=1,Z=1))
    bParentInitial=False
    bPropagateToChildren=True
    Value.Rotation=&RigVMFunction_MathQuaternionFromEuler(Euler=(X=-57,Y=0,Z=0)).Result
    Weight=@n28.OutputParam
   ) {template="Set Relative Transform::Execute(in Child,in Parent,in bParentInitial,in Value,in Weight,in bPropagateToChildren)"}
   n29 = &RigUnit_SphericalPoseReader(
    ActiveRegionScaleFactors=(PositiveWidth=1,NegativeWidth=1,PositiveHeight=1.3,NegativeHeight=1)
    ActiveRegionSize=0.1
    Debug=(bDrawDebug=false,bDraw2D=false,bDrawLocalAxes=false,DebugScale=25,DebugSegments=20,DebugThickness=0.25)
    DriverAxis=(X=-1,Y=0,Z=0)
    DriverItem=(Type=Bone,Name="upperarm_r")
    FalloffRegionScaleFactors=(PositiveWidth=1,NegativeWidth=1,PositiveHeight=0.9,NegativeHeight=1)
    FalloffSize=0.2
    OptionalParentItem=(Type=Bone,Name="spine_05")
    RotationOffset=(X=-60,Y=30,Z=40)
    ExecutePin=@SequenceAggregate_1_1_2_1_1_1_1.B
   )
   > &RigUnit_SetRelativeTransformForItem(
    Child=(Type=Bone,Name="extra_front_plate_windshield_r")
    Parent=(Type=Bone,Name="extra_front_plate_windshield_r")
    Value=(Rotation=(X=0,Y=0,Z=0,W=1),Translation=(X=-1,Y=0,Z=0),Scale3D=(X=1,Y=1,Z=1))
    bParentInitial=False
    bPropagateToChildren=True
    Value.Rotation=&RigVMFunction_MathQuaternionFromEuler(Euler=(X=-57,Y=0,Z=0)).Result
    Weight=@n29.OutputParam
   ) {template="Set Relative Transform::Execute(in Child,in Parent,in bParentInitial,in Value,in Weight,in bPropagateToChildren)"}
   SequenceAggregate_1_1_2_1_1_2 = RigVMAggregateNode(ExecuteContext=@SequenceAggregate.H)
   n30 = &RigVMFunction_MathDoubleRemap(
    SourceMaximum=-16
    SourceMinimum=-4
    TargetMaximum=1
    TargetMinimum=0
    Value=&RigVMFunction_MathQuaternionToEuler(
     RotationOrder=XZY
     Value=&RigUnit_GetTransform(Item=(Type=Bone,Name="spine_01"), Space=LocalSpace, bInitial=false).Transform.Rotation
    ).Result.Z
   )
   &RigUnit_SetRelativeTransformForItem(
    Child=(Type=Bone,Name="extra_pelvis_armor_back_l")
    Parent=(Type=Bone,Name="extra_pelvis_armor_back_l")
    Value=(Rotation=(X=0,Y=0,Z=0,W=1),Translation=(X=1.3,Y=0,Z=0),Scale3D=(X=1,Y=1,Z=1))
    bParentInitial=False
    bPropagateToChildren=True
    Weight=@n30.Result
    ExecutePin=@SequenceAggregate_1_1_2_1_1_2.A
   ) {template="Set Relative Transform::Execute(in Child,in Parent,in bParentInitial,in Value,in Weight,in bPropagateToChildren)"}
   &RigUnit_SetRelativeTransformForItem(
    Child=(Type=Bone,Name="extra_pelvis_armor_back_r")
    Parent=(Type=Bone,Name="extra_pelvis_armor_back_r")
    Value=(Rotation=(X=0,Y=0,Z=0,W=1),Translation=(X=-1.3,Y=0,Z=0),Scale3D=(X=1,Y=1,Z=1))
    bParentInitial=False
    bPropagateToChildren=True
    Weight=@n30.Result
    ExecutePin=@SequenceAggregate_1_1_2_1_1_2.B
   ) {template="Set Relative Transform::Execute(in Child,in Parent,in bParentInitial,in Value,in Weight,in bPropagateToChildren)"}
   &RigUnit_SetRelativeTransformForItem(
    Child=(Type=Bone,Name="extra_pelvis_armor_back")
    Parent=(Type=Bone,Name="extra_pelvis_armor_back")
    Value=(Rotation=(X=0,Y=0,Z=0,W=1),Translation=(X=1,Y=0,Z=0),Scale3D=(X=1,Y=1,Z=1))
    bParentInitial=False
    bPropagateToChildren=True
    Value.Rotation=&RigVMFunction_MathQuaternionFromEuler(Euler=(X=0,Y=0,Z=-19)).Result
    Weight=@n30.Result
    ExecutePin=@SequenceAggregate_1_1_2_1_1_2.C
   ) {template="Set Relative Transform::Execute(in Child,in Parent,in bParentInitial,in Value,in Weight,in bPropagateToChildren)"}
   RigUnit_SetTransform = &RigUnit_SetTransform(
    ExecutePin=()
    Item=(Type=Bone,Name="lowerarm_l")
    Weight=1
    bInitial=False
    Value=&RigUnit_GetTransform(Item=(Type=Control,Name="NewControl")).Transform
    Item.Name=RigVMVariableNode() {variable=temp_bone_control_1}.Value
   )
   > &RigUnit_SetTransform(
    Item=(Type=Bone,Name="lowerarm_l")
    Weight=1
    bInitial=False
    Value=&RigUnit_GetTransform(Item=(Type=Control,Name="NewControl_2")).Transform
    Item.Name=RigVMVariableNode() {variable=temp_bone_control_2}.Value
   )
   n31 = RigVMVariableNode() {variable=temp_bone_control_1}
   > RigVMVariableNode(Value=&RigVMFunction_MathQuaternionToEuler(
     RotationOrder=XZY
     Value=&RigUnit_GetTransform(Item=(Type=Bone,Name="clavicle_r"), Space=LocalSpace, Item.Name=@n31.Value).Transform.Rotation
    ).Result.Z) {variable=temp_1}
   > /Script/ControlRig.RigUnit_DebugHierarchy(
    Color=(R=1,G=1,B=1,A=1)
    DebugDrawSettings=(DepthPriority=SDPG_Foreground,Lifetime=-1)
    Items=()
    Scale=10
    WorldOffset=(Rotation=(X=0,Y=0,Z=0,W=1),Translation=(X=0,Y=0,Z=0),Scale3D=(X=1,Y=1,Z=1))
    bEnabled=True
   ) {is=struct RigVMUnitNode}
  ]}
 SequenceAggregate_1_1_2_1_1_2_1_ContainedGraph {nodes [
   n1 = RigVMFunctionEntryNode()
   SequenceAggregate_1_1_2_1_1_2 = &RigVMFunction_Sequence(ExecuteContext=@n1.ExecuteContext)
   SequenceAggregate_1_1_2_1_1_2_1 = &RigVMFunction_Sequence(ExecuteContext=@SequenceAggregate_1_1_2_1_1_2.B)
   RigVMFunctionReturnNode(A=@SequenceAggregate_1_1_2_1_1_2.A, B=@SequenceAggregate_1_1_2_1_1_2_1.A, C=@SequenceAggregate_1_1_2_1_1_2_1.B)
  ]}
 SequenceAggregate_2_ContainedGraph {nodes [
   n1 = RigVMFunctionEntryNode()
   SequenceAggregate = &RigVMFunction_Sequence(ExecuteContext=@n1.ExecuteContext)
   SequenceAggregate_1 = &RigVMFunction_Sequence(ExecuteContext=@SequenceAggregate.B)
   SequenceAggregate_2 = &RigVMFunction_Sequence(ExecuteContext=@SequenceAggregate_1.B)
   SequenceAggregate_3 = &RigVMFunction_Sequence(ExecuteContext=@SequenceAggregate_2.B)
   SequenceAggregate_4 = &RigVMFunction_Sequence(ExecuteContext=@SequenceAggregate_3.B)
   SequenceAggregate_5 = &RigVMFunction_Sequence(ExecuteContext=@SequenceAggregate_4.B)
   SequenceAggregate_6 = &RigVMFunction_Sequence(ExecuteContext=@SequenceAggregate_5.B)
   SequenceAggregate_7 = &RigVMFunction_Sequence(ExecuteContext=@SequenceAggregate_6.B)
   SequenceAggregate_8 = &RigVMFunction_Sequence(ExecuteContext=@SequenceAggregate_7.B)
   RigVMFunctionReturnNode(
    A=@SequenceAggregate.A
    B=@SequenceAggregate_1.A
    C=@SequenceAggregate_2.A
    D=@SequenceAggregate_3.A
    E=@SequenceAggregate_4.A
    F=@SequenceAggregate_5.A
    G=@SequenceAggregate_6.A
    H=@SequenceAggregate_7.A
    I=@SequenceAggregate_8.A
    J=@SequenceAggregate_8.B
   )
  ]}
]
hierarchy [
 root {loc=<0 0 0>, rot=<0 0 0 1>}
 pelvis {parent=root, loc=<0 -0.7321 92.4218>, rot=<0.0127 0.707 -0.0127 -0.707>}
 spine_01 {parent=pelvis, loc=<5.6002 -0.5128 0>, rot=<0 0 0.034 -0.9994>}
 spine_02 {parent=spine_01, loc=<8.0753 0.0139 0>, rot=<0 0 0.0238 -0.9997>}
 spine_03 {parent=spine_02, loc=<7.0393 0 0>, rot=<0 0 0.0149 -0.9999>}
 spine_04 {parent=spine_03, loc=<8.1024 0 0>, rot=<0 0 -0.0132 -0.9999>}
 spine_05 {parent=spine_04, loc=<10.0146 0 0>, rot=<0 0 -0.0015 -1>}
 clavicle_l {parent=spine_05, loc=<2.5856 -0.2866 -4.3777>, rot=<0.0465 -0.7289 -0.009 -0.6829>}
 upperarm_l {parent=clavicle_l, loc=<18.2264 0.1342 -0.001>, rot=<0.0253 -0.2693 0.0013 -0.9627>}
 lowerarm_l {parent=upperarm_l, loc=<29.0071 0.1686 -1.239>, rot=<0.0391 -0.0099 -0.2453 -0.9686>}
 hand_l {parent=lowerarm_l, loc=<25.8607 0.222 1.0274>, rot=<-0.7193 -0.0988 0.1193 -0.6773>}
 weapon_l {parent=hand_l, loc=<-0.1974 -3.4111 0.3811>, rot=<0 0 0 1>}
 thumb_01_l {parent=hand_l, loc=<1.3118 -0.3191 -2.459>, rot=<0.6546 -0.4042 0.0664 -0.6354>}
 thumb_02_l {parent=thumb_01_l, loc=<5.9336 -0.0448 0.0401>, rot=<-0.0037 0.0213 0.298 -0.9543>}
 thumb_03_l {parent=thumb_02_l, loc=<2.2223 0.3682 -0.0907>, rot=<0.0118 -0.013 0.0066 -0.9998>}
 index_metacarpal_l {parent=hand_l, loc=<1.6535 0.0183 -2.96>, rot=<0.0158 -0.0124 -0.0173 -0.9996>}
 index_01_l {parent=index_metacarpal_l, loc=<8.3008 0.0357 -0.0524>, rot=<-0.0044 0.0485 0.0924 -0.9945>}
 index_02_l {parent=index_01_l, loc=<3.3587 -0.3004 0.0452>, rot=<-0.002 0.0027 0.3225 -0.9466>}
 index_03_l {parent=index_02_l, loc=<2.5578 -0.123 -0.0252>, rot=<-0.002 0.0067 0.2558 -0.9667>}
 middle_metacarpal_l {parent=hand_l, loc=<2.1856 -0.1725 -0.5984>, rot=<0.006 0.0168 -0.0329 -0.9993>}
 middle_01_l {parent=middle_metacarpal_l, loc=<7.3275 -0.2447 0.0129>, rot=<-0.0208 0.0042 0.0858 -0.9961>}
 middle_02_l {parent=middle_01_l, loc=<4.0123 0.1401 0.1175>, rot=<-0.0218 0.0258 0.3512 -0.9357>}
 middle_03_l {parent=middle_02_l, loc=<2.6077 0.0534 0.1373>, rot=<0.0082 0.0188 0.1708 -0.9851>}
 ring_metacarpal_l {parent=hand_l, loc=<2.2074 -0.3625 1.3638>, rot=<-0.0284 0.0722 -0.0458 -0.9959>}
 ring_01_l {parent=ring_metacarpal_l, loc=<7.0295 -0.6197 -0.1446>, rot=<-0.0489 -0.0252 0.1892 -0.9804>}
 ring_02_l {parent=ring_01_l, loc=<3.6174 0.4647 -0.1027>, rot=<0 0 0.2806 -0.9598>}
 ring_03_l {parent=ring_02_l, loc=<2.4152 0.114 0.0199>, rot=<0 0 0.1189 -0.9929>}
 pinky_metacarpal_l {parent=hand_l, loc=<1.7616 -0.9753 3.3814>, rot=<-0.0397 0.0814 -0.069 -0.9935>}
 pinky_01_l {parent=pinky_metacarpal_l, loc=<7.1709 -0.4831 -0.1094>, rot=<-0.0654 -0.0035 0.2242 -0.9723>}
 pinky_02_l {parent=pinky_01_l, loc=<2.8274 0.0893 -0.1043>, rot=<0.0037 0.0011 0.2854 -0.9584>}
 pinky_03_l {parent=pinky_02_l, loc=<2.2046 0.0307 -0.0078>, rot=<-0.0002 0.0098 0.1445 -0.9895>}
 hand_attach_l {parent=hand_l, loc=<8.5 -3 0>, rot=<0.7071 0.7071 0 0>}
 lowerarm_twist_02_l {parent=lowerarm_l, loc=<9.6199 0 0>, rot=<0 0 0 -1>}
 lowerarm_twist_01_l {parent=lowerarm_l, loc=<17.8549 0 0>, rot=<0 0 0 -1>}
 extra_lowerarm_plate_l {parent=lowerarm_l, loc=<2.2638 11.0153 3.1781>, rot=<-0.0698 -0.0046 0.0637 -0.9955>}
 extra_lowerarm_corrective_l {parent=lowerarm_l, loc=<-3.0356 7.9782 2.7375>, rot=<-0.0698 -0.0046 0.0637 -0.9955>}
 extra_rowerarm_exhaustpipe_l {parent=lowerarm_l, loc=<6.5579 -0.4637 6.6424>, rot=<-0.0698 -0.0046 0.0637 -0.9955>}
 upperarm_twist_01_l {parent=upperarm_l, loc=<7.9796 0.0843 -0.6195>, rot=<0 0 0 -1>}
 upperarm_twist_02_l {parent=upperarm_l, loc=<19.5059 0.0843 -0.6195>, rot=<0 0 0 -1>}
 extra_upperarm_pipe_l {parent=upperarm_l, loc=<4.4449 0.5035 11.6958>, rot=<-0.01 -0.0442 -0.0309 -0.9985>}
 extra_upperarm_pipe_end_l {parent=extra_upperarm_pipe_l, loc=<-23.7154 -0.1957 -0.2272>, rot=<0.0107 0.001 -0.0115 -0.9999>}
 extra_upperarm_armor_l {parent=upperarm_l, loc=<0.5623 0.4315 4.4749>, rot=<-0.01 -0.0442 -0.0309 -0.9985>}
 extra_upperarm_front_armor_l {parent=upperarm_l, loc=<0.5623 0.4315 4.4749>, rot=<-0.01 -0.0442 -0.0309 -0.9985>}
 extra_upperarm_side_armor_l {parent=upperarm_l, loc=<-2.8552 -4.8639 10.8352>, rot=<-0.01 -0.0442 -0.0309 -0.9985>}
 extra_upperarm_top_armor_l {parent=upperarm_l, loc=<-6.2712 0.7901 6.5115>, rot=<-0.01 -0.0442 -0.0309 -0.9985>}
 extra_upperarm_top_tweak_l {parent=upperarm_l, loc=<-5.4325 -8.5059 5.0931>, rot=<0.0242 -0.0383 0.6842 -0.7279>}
 attach_cape {parent=spine_05, loc=<9.1509 -13.3783 0>, rot=<0.0616 -0.7044 -0.0616 -0.7044>}
 attach_backpack {parent=spine_05, loc=<-0.123 -15.1423 0>, rot=<-0.0093 -0.707 0.0093 -0.707>}
 neck_01 {parent=spine_05, loc=<10.815 1.9196 0>, rot=<0 0 -0.1864 -0.9825>}
 neck_02 {parent=neck_01, loc=<4.7347 0.3045 0>, rot=<0 0 0.0363 -0.9993>}
 head {parent=neck_02, loc=<5.7632 0.5552 0>, rot=<0 0 0.115 -0.9934>}
 clavicle_r {parent=spine_05, loc=<2.5855 -0.2866 4.3777>, rot=<0.7289 0.0465 -0.6829 0.009>}
 upperarm_r {parent=clavicle_r, loc=<-18.2264 -0.1342 0.001>, rot=<0.0253 -0.2693 0.0013 -0.9627>}
 lowerarm_r {parent=upperarm_r, loc=<-29.0071 -0.1686 1.239>, rot=<0.0391 -0.0099 -0.2453 -0.9686>}
 hand_r {parent=lowerarm_r, loc=<-25.8607 -0.222 -1.0274>, rot=<-0.7193 -0.0988 0.1193 -0.6773>}
 hand_attach_r {parent=hand_r, loc=<-8.5 3 0>, rot=<0 0 -0.7071 -0.7071>}
 weapon_r {parent=hand_r, loc=<0.1974 3.4111 -0.3811>, rot=<0 0 0 1>}
 thumb_01_r {parent=hand_r, loc=<-1.3118 0.3191 2.459>, rot=<0.6546 -0.4042 0.0664 -0.6354>}
 thumb_02_r {parent=thumb_01_r, loc=<-5.9336 0.0448 -0.0401>, rot=<-0.0037 0.0213 0.298 -0.9543>}
 thumb_03_r {parent=thumb_02_r, loc=<-2.2223 -0.3682 0.0907>, rot=<0.0118 -0.013 0.0066 -0.9998>}
 index_metacarpal_r {parent=hand_r, loc=<-1.6535 -0.0183 2.96>, rot=<0.0158 -0.0124 -0.0173 -0.9996>}
 index_01_r {parent=index_metacarpal_r, loc=<-8.3008 -0.0357 0.0524>, rot=<-0.0044 0.0485 0.0924 -0.9945>}
 index_02_r {parent=index_01_r, loc=<-3.3587 0.3004 -0.0452>, rot=<-0.002 0.0027 0.3225 -0.9466>}
 index_03_r {parent=index_02_r, loc=<-2.5578 0.123 0.0252>, rot=<-0.002 0.0067 0.2558 -0.9667>}
 middle_metacarpal_r {parent=hand_r, loc=<-2.1856 0.1725 0.5984>, rot=<0.006 0.0168 -0.0329 -0.9993>}
 middle_01_r {parent=middle_metacarpal_r, loc=<-7.3275 0.2447 -0.0129>, rot=<-0.0208 0.0042 0.0858 -0.9961>}
 middle_02_r {parent=middle_01_r, loc=<-4.0123 -0.1401 -0.1175>, rot=<-0.0218 0.0258 0.3512 -0.9357>}
 middle_03_r {parent=middle_02_r, loc=<-2.6077 -0.0534 -0.1373>, rot=<0.0082 0.0188 0.1708 -0.9851>}
 ring_metacarpal_r {parent=hand_r, loc=<-2.2074 0.3625 -1.3638>, rot=<-0.0284 0.0722 -0.0458 -0.9959>}
 ring_01_r {parent=ring_metacarpal_r, loc=<-7.0295 0.6197 0.1446>, rot=<-0.0489 -0.0252 0.1892 -0.9804>}
 ring_02_r {parent=ring_01_r, loc=<-3.6174 -0.4647 0.1027>, rot=<0 0 0.2806 -0.9598>}
 ring_03_r {parent=ring_02_r, loc=<-2.4152 -0.114 -0.0199>, rot=<0 0 0.1189 -0.9929>}
 pinky_metacarpal_r {parent=hand_r, loc=<-1.7616 0.9753 -3.3814>, rot=<-0.0397 0.0814 -0.069 -0.9935>}
 pinky_01_r {parent=pinky_metacarpal_r, loc=<-7.1709 0.4831 0.1094>, rot=<-0.0654 -0.0035 0.2242 -0.9723>}
 pinky_02_r {parent=pinky_01_r, loc=<-2.8274 -0.0893 0.1043>, rot=<0.0037 0.0011 0.2854 -0.9584>}
 pinky_03_r {parent=pinky_02_r, loc=<-2.2046 -0.0307 0.0078>, rot=<-0.0002 0.0098 0.1445 -0.9895>}
 lowerarm_twist_02_r {parent=lowerarm_r, loc=<-9.6199 0 0>, rot=<0 0 0 -1>}
 lowerarm_twist_01_r {parent=lowerarm_r, loc=<-17.8549 0 0>, rot=<0 0 0 -1>}
 extra_rowerarm_plate_r {parent=lowerarm_r, loc=<-2.2639 -11.0152 -3.1779>, rot=<-0.0698 -0.0046 0.0637 -0.9955>}
 extra_rowerarm_corrective_r {parent=lowerarm_r, loc=<3.0358 -7.9782 -2.7379>, rot=<-0.0698 -0.0046 0.0637 -0.9955>}
 extra_lowerarm_corrective_r {parent=lowerarm_r, loc=<3.0358 -7.9782 -2.7379>, rot=<-0.0698 -0.0046 0.0637 -0.9955>}
 extra_lowerarm_exhaustpipe_r {parent=lowerarm_r, loc=<-6.558 0.4638 -6.6422>, rot=<-0.0698 -0.0046 0.0637 -0.9955>}
 upperarm_twist_01_r {parent=upperarm_r, loc=<-7.9796 -0.0843 0.6195>, rot=<0 0 0 -1>}
 upperarm_twist_02_r {parent=upperarm_r, loc=<-19.5059 -0.0843 0.6195>, rot=<0 0 0 -1>}
 extra_upperarm_pipe_r {parent=upperarm_r, loc=<-4.4449 -0.5035 -11.6958>, rot=<-0.01 -0.0442 -0.0309 -0.9985>}
 extra_upperarm_pipe_end_r {parent=extra_upperarm_pipe_r, loc=<23.7155 0.1957 0.2272>, rot=<0.0107 0.001 -0.0115 -0.9999>}
 extra_upperarm_armor_r {parent=upperarm_r, loc=<-0.5625 -0.4315 -4.4747>, rot=<-0.01 -0.0442 -0.0309 -0.9985>}
 extra_upperarm_front_armor_r {parent=upperarm_r, loc=<-0.5625 -0.4315 -4.4747>, rot=<-0.01 -0.0442 -0.0309 -0.9985>}
 extra_upperarm_side_armor_r {parent=upperarm_r, loc=<2.8552 4.8639 -10.8352>, rot=<-0.01 -0.0442 -0.0309 -0.9985>}
 extra_upperarm_top_armor_r {parent=upperarm_r, loc=<6.2714 -0.7901 -6.5119>, rot=<-0.01 -0.0442 -0.0309 -0.9985>}
 extra_upperarm_top_tweak_r {parent=upperarm_r, loc=<5.4327 8.5059 -5.0934>, rot=<0.0242 -0.0383 0.6842 -0.7279>}
 extra_front_plate_l {parent=spine_05, loc=<1.4351 17.7502 -1.3063>, rot=<0.0727 -0.7034 0.0161 -0.7069>}
 extra_front_plate_windshield_l {parent=extra_front_plate_l, loc=<12.0089 0.1855 6.3836>, rot=<0.2479 -0.0183 -0.0063 -0.9686>}
 extra_front_plate_r {parent=spine_05, loc=<1.4352 17.7503 1.3063>, rot=<-0.7034 -0.0727 0.7069 0.0161>}
 extra_front_plate_windshield_r {parent=extra_front_plate_r, loc=<-12.0088 -0.1854 -6.384>, rot=<0.2479 -0.0183 -0.0063 -0.9686>}
 extra_top_front_plate_l {parent=spine_05, loc=<9.4905 7.9215 -6.1177>, rot=<0.1245 -0.7331 -0.1352 -0.6548>}
 extra_top_back_plate_l {parent=spine_05, loc=<14.0045 -5.8032 -8.7876>, rot=<0.0072 -0.7909 -0.044 -0.6103>}
 extra_top_front_plate_r {parent=spine_05, loc=<9.4907 7.9215 6.1177>, rot=<0.7331 0.1245 -0.6548 0.1352>}
 extra_top_back_plate_r {parent=spine_05, loc=<14.0041 -5.8033 8.7876>, rot=<0.7909 0.0072 -0.6103 0.044>}
 dfrm_clavicle_l {parent=spine_05, loc=<2.5856 -0.2866 -4.3777>, rot=<0.0465 -0.7289 -0.009 -0.6829>}
 dfrm_clavicle_r {parent=spine_05, loc=<2.5855 -0.2866 4.3777>, rot=<0.7289 0.0465 -0.6829 0.009>}
 extra_chest_bar {parent=spine_05, loc=<12.8201 1.8525 0>, rot=<0 0 -0.2637 -0.9646>}
 thigh_l {parent=pelvis, loc=<-1.8716 -0.0109 -11.9936>, rot=<0.0012 0.0572 0.0155 -0.9982>}
 calf_l {parent=thigh_l, loc=<-41.9178 -0.2246 0.0926>, rot=<0.0006 -0.0161 -0.0391 -0.9991>}
 foot_l {parent=calf_l, loc=<-38.7057 1.2438 0.2452>, rot=<0.0975 -0.0474 0.0317 -0.9936>}
 ball_l {parent=foot_l, loc=<-6.1614 13.1469 0.6807>, rot=<0.0001 -0.0001 -0.7063 -0.7079>}
 calf_twist_02_l {parent=calf_l, loc=<-15.9268 -0.0967 0.0397>, rot=<0.0237 0.0095 -0.0035 -0.9997>}
 calf_twist_01_l {parent=calf_l, loc=<-32.9335 -0.0967 0.0397>, rot=<0.0237 0.0095 -0.0035 -0.9997>}
 extra_foot_piston_l {parent=calf_l, loc=<-38.7969 -7.2172 -1.5833>, rot=<-0.0097 0.0187 -0.0863 -0.9961>}
 extra_calf_gastank_l {parent=calf_l, loc=<5.4123 -3.156 -9.2652>, rot=<0.0018 0.0348 -0.0626 -0.9974>}
 extra_calf_wheel_bot_l {parent=calf_l, loc=<-27.7496 -6.6009 -11.9501>, rot=<0.0014 0.042 -0.0626 -0.9972>}
 extra_calf_wheel_top_l {parent=calf_l, loc=<-10.34 -6.9756 -10.4249>, rot=<0.0014 0.042 -0.0626 -0.9972>}
 extra_calf_corrective_l {parent=calf_l, loc=<-3.968 -8.6638 0.2706>, rot=<-0.0182 0.0138 -0.0009 -0.9997>}
 extra_foot_piston_end_l {parent=calf_l, loc=<-30.6899 -5.8052 -1.2626>, rot=<-0.0091 -0.0105 -0.0241 -0.9996>}
 thigh_twist_01_l {parent=thigh_l, loc=<-21.6303 0 0>, rot=<0.0013 0 0 -1>}
 extra_calf_armor_l {parent=thigh_l, loc=<-37.2153 -2.719 0.2713>, rot=<0.0002 -0.0059 -0.0391 -0.9992>}
 extra_thigh_armor_l {parent=thigh_l, loc=<-1.2732 9.8155 0.2797>, rot=<0 0.0154 0 -0.9999>}
 extra_calf_backplate_l {parent=thigh_l, loc=<-37.2153 -2.719 0.2713>, rot=<0.0002 -0.0059 -0.0391 -0.9992>}
 thigh_r {parent=pelvis, loc=<-1.8716 -0.0109 11.9936>, rot=<0.0572 -0.0012 0.9982 0.0155>}
 calf_r {parent=thigh_r, loc=<41.9178 0.2246 -0.0926>, rot=<0.0006 -0.0161 -0.0391 -0.9991>}
 foot_r {parent=calf_r, loc=<38.7057 -1.2438 -0.2452>, rot=<0.0975 -0.0474 0.0317 -0.9936>}
 ball_r {parent=foot_r, loc=<6.1614 -13.1469 -0.6807>, rot=<0.0001 -0.0001 -0.7063 -0.7079>}
 calf_twist_02_r {parent=calf_r, loc=<15.9268 0.0967 -0.0397>, rot=<0.0237 0.0095 -0.0035 -0.9997>}
 calf_twist_01_r {parent=calf_r, loc=<32.9335 0.0967 -0.0397>, rot=<0.0237 0.0095 -0.0035 -0.9997>}
 extra_foot_piston_r {parent=calf_r, loc=<38.7969 7.2173 1.5834>, rot=<-0.0097 0.0187 -0.0863 -0.9961>}
 extra_calf_gastank_r {parent=calf_r, loc=<-5.4123 3.1561 9.2652>, rot=<0.0018 0.0348 -0.0626 -0.9974>}
 extra_calf_wheel_top_r {parent=calf_r, loc=<10.34 6.9756 10.4248>, rot=<0.0014 0.042 -0.0626 -0.9972>}
 extra_calf_wheel_bot_r {parent=calf_r, loc=<27.7496 6.6009 11.95>, rot=<0.0014 0.042 -0.0626 -0.9972>}
 extra_calf_corrective_r {parent=calf_r, loc=<3.968 8.6638 -0.2707>, rot=<-0.0182 0.0138 -0.0009 -0.9997>}
 extra_foot_piston_end_r {parent=calf_r, loc=<30.6899 5.8052 1.2626>, rot=<-0.0091 -0.0105 -0.0241 -0.9996>}
 thigh_twist_01_r {parent=thigh_r, loc=<21.6303 0 0>, rot=<0.0013 0 0 -1>}
 extra_calf_armor_r {parent=thigh_r, loc=<37.2153 2.719 -0.2713>, rot=<0.0002 -0.0059 -0.0391 -0.9992>}
 extra_thigh_armor_r {parent=thigh_r, loc=<1.2732 -9.8155 -0.2798>, rot=<0 0.0154 0 -0.9999>}
 extra_calf_backplate_r {parent=thigh_r, loc=<37.2153 2.719 -0.2713>, rot=<0.0002 -0.0059 -0.0391 -0.9992>}
 extra_pelvis_armor_l {parent=pelvis, loc=<10.3255 -0.4235 -14.7554>, rot=<0.013 0.1044 0.0043 -0.9944>}
 extra_pelvis_armor_r {parent=pelvis, loc=<10.3256 -0.4236 14.7554>, rot=<0.1044 -0.013 0.9944 0.0043>}
 extra_pelvis_armor_back_l {parent=pelvis, loc=<11.2831 -9.4493 -11.3749>, rot=<-0.0099 0.0284 0.6972 -0.7162>}
 extra_pelvis_armor_back_r {parent=pelvis, loc=<11.2829 -9.4493 11.3749>, rot=<-0.0284 -0.0099 -0.7162 -0.6972>}
 extra_pelvis_armor_back {parent=pelvis, loc=<10.1377 -12.0305 0>, rot=<0 0 0.9558 -0.2939>}
 NewControl {controlType=EULER_TRANSFORM, type=Control, loc=<0 0 0>, rot=<0 0 0 1>}
 NewControl_2 {controlType=EULER_TRANSFORM, type=Control, loc=<0 0 0>, rot=<0 0 0 1>}
]
variables [
 temp_1 {type=double, default=0}
 temp_2 {type=double, default=0}
 temp_bone_control_1 {type=FName, default=upperarm_l}
 temp_bone_control_2 {type=FName, default=upperarm_r}
]
