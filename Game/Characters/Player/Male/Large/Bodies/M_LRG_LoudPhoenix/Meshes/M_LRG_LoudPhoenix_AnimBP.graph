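# Animation Blueprint for the LoudPhoenix large male body that follows the attached parent mesh's pose and adds part-specific adjustments.
# An additive attach_backpack offset eases in and out with skydive floating or an open parachute, then feeds a character-specific Control Rig.
# DefaultSlot animation is layered onto right-arm pipe and corrective bones, weighted by the emote_BuiltIn curve from the front-end or player anim instance.
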AnimBlueprint /Game/Characters/Player/Male/Large/Bodies/M_LRG_LoudPhoenix/Meshes/M_LRG_LoudPhoenix_AnimBP
target_skeleton=&/M_LRG_LoudPhoenix_Skeleton
graphs [
 AnimGraph {nodes [
   n1 = "Transform (Modify) Bone - Bone: attach_backpack"(
    ComponentPose=Local To Component(LocalPose=Copy Pose From Mesh() {bUseAttachedParent=true})
    Alpha=Get skydive_float_paracute_alpha() -> (skydive_float_paracute_alpha:double)
    Translation="0,-2,3"
    Rotation="0,0,0"
    Scale="1,1,1"
   ) {translationMode=BMM_Additive, alphaBoolBlend {blendOption=Linear}, boneToModify {boneName=attach_backpack}, translation {y=-2, z=3}}
   n2 = Control Rig(Source=Component To Local(ComponentPose=@n1)) {
    alphaBoolBlend {blendOption=Linear}
    controlRigAssetReference {blueprintRigClass=&/M_LRG_LoudPhoenix_CtrlRig.M_LRG_LoudPhoenix_CtrlRig_C}
   }
   n3 = Layered blend per bone(BasePose=@n2, BlendPoses_0=Slot 'DefaultSlot'(), BlendWeights_0=Get BuiltIn Emote() -> (BuiltIn Emote:double)) {layerSetup [{branchFilters [{boneName=extra_lowerarm_exhaustpipe_r}, {boneName=extra_lowerarm_corrective_r}, {boneName=extra_upperarm_pipe_r}]}]}
   Output Pose(Result=@n3) {layerGroup=None}
  ]}
 EventGraph {nodes [
   n1 = Event BlueprintUpdateAnimation()
   > Set DeltaTimeX(DeltaTimeX:double=@n1.DeltaTimeX) -> (Output_Get:double)
   > n2 = Sequence()
   > n3 = Branch(Condition=OR Boolean(A=Get bIsSkydiveFloating(), B=Get bIsParachuteOpened()))
   > Set skydive_float_paracute_alpha(skydive_float_paracute_alpha:double=FInterpTo_Constant!(
     Current:double=Get skydive_float_paracute_alpha()
     Target:double=1
     DeltaTime:double=Get DeltaTimeX() -> (DeltaTimeX:double)
     InterpSpeed:double=2
    ) -> (ReturnValue:double)) -> (Output_Get:double)
   Set skydive_float_paracute_alpha(
    execute=@n3.else
    skydive_float_paracute_alpha=FInterpTo_Constant(Current=Get skydive_float_paracute_alpha(), Target=0, DeltaTime=Get DeltaTimeX(), InterpSpeed=2)
   )
   n4 = Branch(execute=@n2.then_1, Condition=Get bIsInFrontEnd())
   > Set BuiltIn Emote(BuiltIn Emote:double=GetCurveValue(self=GetOwnerFrontEndAnimInstance(), CurveName=emote_BuiltIn)) -> (Output_Get:double)
   Set BuiltIn Emote(
    execute=@n4.else
    BuiltIn Emote=GetCurveValue(
     self=GetOwnerFortPlayerAnimInstance!(self:Custom Character Part Anim Instance Object Reference) -> (ReturnValue:Fort Player Anim Instance Object Reference)
     CurveName=emote_BuiltIn
    )
   )
  ]}
]
variables [
 BuiltIn Emote {type=number}
 DeltaTimeX {type=number}
 skydive_float_paracute_alpha {type=number}
]
